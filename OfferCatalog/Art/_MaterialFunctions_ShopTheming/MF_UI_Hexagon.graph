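# UI material function for offer catalog shop theming that outputs an animated hexagonal haze effect with an alpha.
# A hexagon SDF and radial angle form hex-polar coordinates, which cloud noise distorts and the time input scrolls into repeating sharp bands.
# These bands mask desaturated portal haze, and a faint scrolling hex-pulse pattern is added on top.
# Default time uses FPS-quantized stepping; alpha comes from the haze texture's red and blue channels.
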MaterialFunction /OfferCatalog/Art/_MaterialFunctions_ShopTheming/MF_UI_Hexagon
graph {nodes [
  Constant2Vector() {g=1, r=1}
  n1 = TextureCoordinate()
  FunctionInput(Preview=@n1) {bUsePreviewValueAsDefault=true, inputName=UV_Static, inputType=FunctionInput_Vector2}
  n2 = FunctionInput(Preview=@n1) {bUsePreviewValueAsDefault=true, inputName=UV, inputType=FunctionInput_Vector2}
  n3 = TextureSample!(UVs=@n2, Tex, Bias, Apply View MipBias) {mipValueMode=TMVM_MipBias, texture=/OfferCatalog/Art/FearlessFlight/T_UI_PortalHaze}
  n4 = Desaturation(None=@n3.RGB, Fraction=Constant() {r=1})
  n5 = Constant() {r=0.5}
  n6 = Constant() {r=0.0833}
  n7 = AppendVector(
   A=MaterialFunctionCall!(Corner roundness, UVs=@n2, Radius=Constant() {r=1}, Glow Min, Glow Max, Outline Thickness, Is stroke thickness relative to size) {materialFunction=/OfferCatalog/Art/FearlessFlight/MF_UI_HexagonSDF}.SDF Outline
   B=MaterialFunctionCall!(Vector or UVs=MaterialFunctionCall(UVs=@n2, "Rotation Angle (0-1)"=@n6) {materialFunction=&CustomRotator}.Rotated Values) {materialFunction=/FortUILibrary/MaterialFunctions/Basics/MF_UI_CheaperVectorToRadial}.Vector Converted to Angle
  )
  n8 = ComponentMask(None=@n7) {r=true}
  n9 = ComponentMask(None=@n7) {g=true}
  n10 = NamedRerouteUsage()
  n11 = FunctionInput() {bUsePreviewValueAsDefault=true, inputName=time_gain, inputType=FunctionInput_Vector2, previewValue=<0.25 0 0 1>}
  n12 = FunctionInput(Preview=MaterialFunctionCall(Y=Multiply(A=@n10, B=@n11) {constB=0.1}) {materialFunction=&MakeFloat2}.Result) {bUsePreviewValueAsDefault=true, inputName=time, inputType=FunctionInput_Vector2}
  n13 = ComponentMask(None=@n12) {g=true}
  n14 = Saturate(None=@n13)
  n15 = LinearInterpolate(A=@n9, B=OneMinus(None=@n9), Alpha=@n14)
  n16 = ComponentMask(None=@n7) {g=true, r=true}
  n17 = Constant2Vector() {g=1, r=2}
  n18 = Subtract(A=TextureSample!(UVs=Multiply(A=@n16, B=@n17), Tex, Apply View MipBias) {texture=/EpicBaseTextures/Noises/T_Clouds_Noise3_D}.R) {constB=0.5}
  n19 = AppendVector(A=@n8, B=Add(A=@n15, B=Multiply(A=@n18) {constB=0.1}))
  n20 = Abs(None=Subtract(A=@n19, B=@n12))
  n21 = Add(A=@n19, B=@n12)
  n22 = Frac(None=Multiply(A=LinearInterpolate(A=@n20, B=@n21, Alpha=@n14)) {constB=2.71})
  n23 = OneMinus(None=Saturate(None=ComponentMask(None=@n22) {g=true}))
  n24 = Saturate(None=Multiply(
    A=MaterialFunctionCall(In=@n4, Contrast=@n5) {materialFunction=&CheapContrast}.Result
    B=Power(Base=@n23) {constExponent=12}
   ))
  n25 = ComponentMask(None=Frac(None=NamedRerouteUsage())) {g=true}
  n26 = AppendVector(A=NamedRerouteUsage(), B=OneMinus(None=Saturate(None=@n25)))
  n27 = Constant2Vector() {g=8, r=2}
  n28 = FunctionInput() {bUsePreviewValueAsDefault=true, inputName=PatternHexIntensity, inputType=FunctionInput_Scalar, previewValue=<0.015 0 0 1>}
  n29 = Multiply(
   A=TextureSample(UVs=Multiply(A=@n26, B=@n27)) {samplerType=SAMPLERTYPE_Masks, texture=/Game/Creative/Sets/PropSets/Primitives/Customization/Textures/Hex/T_CM_HexPulse_RandomColor}.RGB
   B=@n28
  ) {constB=0.015}
  FunctionOutput(None=Max(A=Add(A=@n24, B=@n29)) {constB=0})
  FunctionOutput(None=Max(A=Max(A=@n3.R, B=@n3.R), B=@n3.B)) {outputName=Alpha, sortPriority=1}
  n30 = FunctionInput() {bUsePreviewValueAsDefault=true, inputName=flip, inputType=FunctionInput_Scalar, previewValue=<1 0 0 1>}
  Multiply(A=@n12, B=@n30)
  n31 = ScalarParameter() {defaultValue=12, parameterName=FPS}
  n32 = Multiply(A=Time(), B=@n31) {constB=12}
  NamedRerouteDeclaration(None=Divide(A=Floor(None=@n32), B=@n31) {constB=12}) {name=fps, nodeColor=<rgba 1 0 0.4688 1>}
  NamedRerouteDeclaration(None=@n8) {name=HexagonSDF-outeru, nodeColor=<rgba 0 0.6719 1 1>}
  n33 = AppendVector(A=@n8, B=@n15)
  n34 = FunctionInput() {bUsePreviewValueAsDefault=true, inputName=hexTimeGain, inputType=FunctionInput_Scalar, previewValue=<0.5 0 0 1>}
  n35 = Multiply(A=@n12, B=@n34) {constB=0.5}
  n36 = Subtract(A=@n33, B=@n35)
  NamedRerouteDeclaration(None=LinearInterpolate(A=Abs(None=@n36), B=Add(A=@n33, B=@n35), Alpha=@n14)) {name=NormalUVs, nodeColor=<rgba 0.0547 0 1 1>}
  Sign(None=@n13)
  Time()
 ]}
