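# MF_UI_ScalePosition is a material function in the OfferCatalog shop theming set that scales and repositions UV coordinates.
# It scales the input UVs about their center with ScaleUVsByCenter, then adds an offset built by appending the X Position and Y Position inputs.
# The offset is applied after scaling, so it is not affected by the Scale value.
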MaterialFunction /OfferCatalog/Art/_MaterialFunctions_ShopTheming/MF_UI_ScalePosition
graph {nodes [
  n1 = FunctionInput() {inputName=UVs, inputType=FunctionInput_Vector2, sortPriority=1}
  n2 = FunctionInput(Preview=Constant() {r=1}) {inputName=Scale, inputType=FunctionInput_Scalar, sortPriority=2}
  n3 = FunctionInput(Preview=Constant()) {inputName=X Position, inputType=FunctionInput_Scalar, sortPriority=3}
  n4 = FunctionInput(Preview=Constant()) {inputName=Y Position, inputType=FunctionInput_Scalar, sortPriority=4}
  FunctionOutput(None=Add(
    A=MaterialFunctionCall(UVs=@n1, Texture Scale=@n2) {materialFunction=&ScaleUVsByCenter}.UVs
    B=AppendVector(A=@n3, B=@n4)
   ))
 ]}
